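# Physics asset for the face accessory part of the SharpMagnet medium male player cosmetic, covering six dynamic hair bones.
# Each hair bone has a small capsule body, with a smaller capsule on the head as their shared anchor.
# All six hair bodies constrain directly to the head rather than chaining, with identical moderate swing limits and a tight twist limit.
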PhysicsAsset /BRCosmetics/Characters/Player/Male/Medium/Bodies/M_MED_SharpMagnet/Meshes/Parts/M_MED_SharpMagnet_FaceAcc_Physics
bodies {
 dyn_hair_01 [{shapeType=Capsule, length=1, radius=1, shapeName=dyn_hair_01_capsule, center=<1 0 0>, rotation=<rot 90 0 0>}]
 dyn_hair_02 [{shapeType=Capsule, length=1, radius=1, shapeName=dyn_hair_02_capsule, center=<1 0 0>, rotation=<rot 90 0 0>}]
 dyn_hair_03 [{shapeType=Capsule, length=1, radius=1, shapeName=dyn_hair_03_capsule, center=<1 0 0>, rotation=<rot 90 0 0>}]
 dyn_hair_04 [{shapeType=Capsule, length=1, radius=1, shapeName=dyn_hair_04_capsule, center=<1 0 0>, rotation=<rot 90 0 0>}]
 dyn_hair_05 [{shapeType=Capsule, length=1, radius=1, shapeName=dyn_hair_05_capsule, center=<1 0 0>, rotation=<rot 90 0 0>}]
 dyn_hair_06 [{shapeType=Capsule, length=1, radius=1, shapeName=dyn_hair_06_capsule, center=<1 0 0>, rotation=<rot 90 0 0>}]
 head [{shapeType=Capsule, length=0.505, radius=0.505, center=<2 0 0>}]
}
constraints [
 {bone1Name=dyn_hair_01, bone2Name=head, swing1LimitDegrees=45, swing2LimitDegrees=45, twistLimitDegrees=3}
 {bone1Name=dyn_hair_02, bone2Name=head, swing1LimitDegrees=45, swing2LimitDegrees=45, twistLimitDegrees=3}
 {bone1Name=dyn_hair_03, bone2Name=head, swing1LimitDegrees=45, swing2LimitDegrees=45, twistLimitDegrees=3}
 {bone1Name=dyn_hair_04, bone2Name=head, swing1LimitDegrees=45, swing2LimitDegrees=45, twistLimitDegrees=3}
 {bone1Name=dyn_hair_05, bone2Name=head, swing1LimitDegrees=45, swing2LimitDegrees=45, twistLimitDegrees=3}
 {bone1Name=dyn_hair_06, bone2Name=head, swing1LimitDegrees=45, swing2LimitDegrees=45, twistLimitDegrees=3}
]
